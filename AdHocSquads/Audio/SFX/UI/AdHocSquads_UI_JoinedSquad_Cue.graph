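# Sound cue for the UI sound that plays when joining a squad in the AdHocSquads feature, judging by its name.
# A single wave player feeds the AdHocSquads_UI_JoinedSquad sound wave straight to the output, with no randomization or modulation.
# It routes through the critical in-game UI sound class and sets a very large max distance.
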SoundCue /AdHocSquads/Audio/SFX/UI/AdHocSquads_UI_JoinedSquad_Cue
root=WavePlayer_1
nodes [{id=WavePlayer_1, properties {soundWaveAssetPtr=&/AdHocSquads_UI_JoinedSquad}}]
properties {maxDistance=2097152, soundClassObject=&Fort_UI_InGame_Critical, volumeMultiplier=1}
